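# Translucent crystal material for the NickelScan back bling, wrapping crystal-specific attributes in the 2023 character parent PreFX/PostFX functions.
# Pre-skinned-position planar projections build crack, noise and distortion masks, alongside parallax, shine, thin-film and fake refraction terms.
# An HLSL two-stop gradient colors the glow, with a static switch substituting a Galaxy effect with a screen-space hollow.
# Emissive scales by time of day, simplifies on mobile, and opacity uses sorted translucency with quality-dependent depth fade.
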Material /BRCosmetics/Accessories/FORT_Backpacks/Backpack_M_MED_NickelScan/Materials/M_NickelScan_Crystal
graph {
 properties {
  allowTranslucentCustomDepthWrites=true
  bUseMaterialAttributes=true
  bUsedWithMeshDeformer=true
  bUsedWithMorphTargets=true
  bUsedWithSkeletalMesh=true
  blendMode=BLEND_Translucent
  shadingModel=MSM_FromMaterialExpression
  translucencyLightingMode=TLM_SurfacePerPixelLighting
  usedShadingModels=Default Lit | Subsurface | Clear Coat | Subsurface Profile | Cloth
 }
 nodes [
  CameraPositionWS()
  n1 = MaterialFunctionCall() {materialFunction=&MF_CharacterParent_2023_PreFX}
  MaterialFunctionCall(Material=NamedRerouteUsage(), M=@n1.M, Pre Skinned Local Position=@n1.M, Distance Blend=@n1.M, Fresnel=@n1.M) {materialFunction=&MF_CharacterParent_2023_PostFX}
  NamedRerouteDeclaration(None=@n1.Tech Art Mask) {name=TechArtMask, nodeColor=<rgba 1 0.6797 0 1>}
  n2 = Divide(
   A=NamedRerouteUsage()
   B=ScalarParameter() {defaultValue=20, group=NickelScan_Crystal, parameterName=Crystal_CrackNoiseScale, sortPriority=50}
  )
  n3 = NamedRerouteUsage()
  n4 = NamedRerouteUsage()
  n5 = Add(A=Add(A=ComponentMask(None=@n2) {b=true, r=true}, B=@n3), B=@n4)
  n6 = TextureObject() {texture=/Game/Global/Textures/Default/Thinfilm/T_F_MED_Global_FB_ThinFilm_color}
  n7 = Add(A=Add(A=ComponentMask(None=@n2) {b=true, g=true}, B=@n3), B=@n4)
  n8 = NamedRerouteUsage()
  n9 = ComponentMask(None=@n8) {r=true}
  n10 = Constant() {r=1}
  n11 = ComponentMask(None=MaterialFunctionCall(In=Abs(None=@n9), Contrast=@n10) {materialFunction=&CheapContrast}.Result) {r=true}
  n12 = Add(A=ComponentMask(None=@n2) {g=true, r=true}, B=@n3)
  n13 = ComponentMask(None=@n8) {b=true}
  n14 = ComponentMask(None=MaterialFunctionCall(In=Abs(None=@n13), Contrast=@n10) {materialFunction=&CheapContrast}.Result) {r=true}
  n15 = LinearInterpolate(
   A=LinearInterpolate(
    A=TextureSample(UVs=@n5, Tex=@n6) {samplerSource=SSM_Wrap_WorldGroupSettings}.R
    B=TextureSample(UVs=@n7, Tex=@n6) {samplerSource=SSM_Wrap_WorldGroupSettings}.R
    Alpha=@n11
   ) {desc=Float3}
   B=TextureSample(UVs=Add(A=@n12, B=@n4), Tex=@n6) {samplerSource=SSM_Wrap_WorldGroupSettings}.R
   Alpha=@n14
  )
  n16 = ScalarParameter() {defaultValue=1, group=NickelScan_Crystal, parameterName=Crystal_NoiseIntensity, sortPriority=50}
  NamedRerouteDeclaration(None=Multiply(A=Multiply(A=@n15, B=@n16), B=NamedRerouteUsage())) {name=CrystalMatter, nodeColor=<rgba 0 0.0859 1 1>}
  NamedRerouteDeclaration(None=VertexInterpolator(VS=PreSkinnedNormal()).PS) {name=PreSkinnedNormal, nodeColor=<rgba 0 1 0.6719 1>}
  n17 = VectorParameter() {group=NickelScan_Crystal, parameterName=Crystal_BaseEmissiveColor, sortPriority=60, channelNames {a=DeSaturation, b="", g="", r=""}}
  n18 = TextureSample() {samplerSource=SSM_Wrap_WorldGroupSettings, texture=/Game/Global/Textures/Default/Masks_and_Gradients/T_MF_Crystal}
  n19 = GetMaterialAttributes(None=MaterialFunctionCall!(Tech Art Mask=NamedRerouteUsage(), Mask=NamedRerouteUsage(), In=GetMaterialAttributes().MaterialAttributes) {materialFunction=/Game/Global/Materials/Character_Parent_2023/CrewStyle_Parent_2023/Galaxy/MF_CrewStyle_Galaxy_Inst}.Result) {attributeGetTypes [#1]}
  n20 = TransformPosition(None=Constant4Vector() {constant=<rgba 1 -9.5 -1.8 1>}.RGB) {transformType=TRANSFORMPOSSOURCE_World}
  n21 = SphereMask(
   A=Subtract(A=MaterialFunctionCall!(Position=@n20, UVRadialScale) {materialFunction=/Game/Characters/Player/Common/MaterialFunctions/Superlevel/CharacterStabilizedScreenSpaceUVs}.StabilizedUVs) {constB=0.5}
   B=Constant()
   Radius=ScalarParameter() {defaultValue=0.028, parameterName=GalaxyHollow_Radius}
   Hardness=ScalarParameter() {defaultValue=0.3, parameterName=GalaxyHollow_Hardness}
  ) {attenuationRadius=0.05, hardnessPercent=1}
  n22 = LinearInterpolate(
   A=MaterialFunctionCall(Distance=ScalarParameter() {defaultValue=600, group=NickelScan_Crystal, parameterName=GalaxyHollow_CamFadeDist, sortPriority=20}) {materialFunction=&Near_Camera_Fade}.Result
   Alpha=Power(Base=OneMinus(None=@n21))
  ) {constA=1}
  n23 = VectorParameter() {
   group=NickelScan_Crystal
   parameterName=Crystal_GlowColor_A
   sortPriority=60
   channelNames {a=Postion, b="", g="", r=""}
   defaultValue=<rgba 1 1 1 0>
  }
  MaterialExpressionStaticSwitchParameter_1 = StaticSwitchParameter(True=AppendVector(A=Multiply(A=@n19.Emissive Color, B=@n22), B=Constant()), False=@n23.RGBA) {parameterName=Galaxy}
  MaterialExpressionVectorParameter_3 = VectorParameter() {
   group=NickelScan_Crystal
   parameterName=Crystal_GlowColor_B
   sortPriority=60
   channelNames {a=Postion, b="", g="", r=""}
   defaultValue=<rgba 0 0 0 1>
  }
  MaterialExpressionNamedRerouteUsage_5 = NamedRerouteUsage()
  n24 = Custom!(
   ColorA=@MaterialExpressionStaticSwitchParameter_1
   ColorB=@MaterialExpressionVectorParameter_3.RGBA
   Gradient=@MaterialExpressionNamedRerouteUsage_5
  ) {
   code="float t = saturate(1-Gradient);\r\n// Extract the positions from the Alpha channels\r\nfloat posA = ColorA.a;\r\nfloat posB = ColorB.a;\r\nfloat posEnd = 1.0;\r\nfloat3 ColorEnd = (0.0,0.0,0.0);\r\n// Calculate the transition masks (creating a 0 to 1 range between each custom position)\r\n// max(..., 1e-5) prevents division-by-zero errors if two stops share the exact same position\r\nfloat mask1 = saturate((t - posA) / max(posB - posA, 1e-5));\r\nfloat mask2 = saturate((t - posB) / max(posEnd - posB, 1e-5));\r\n// Blend the colors using the masks\r\nfloat3 blendAB = lerp(ColorA.rgb, ColorB.rgb, mask1);\r\nreturn lerp(blendAB, ColorEnd, mask2);"
   desc=Crystal_GlowColor
   description=Lerp2Color
   inputs [
    {
     inputName=ColorA
     input {expression=&self:MaterialExpressionStaticSwitchParameter_1}
    }
    {
     inputName=ColorB
     input {
      expression=&self:MaterialExpressionVectorParameter_3
      mask=1
      maskA=1
      maskB=1
      maskG=1
      maskR=1
      outputIndex=5
     }
    }
    {
     inputName=Gradient
     input {expression=&self:MaterialExpressionNamedRerouteUsage_5}
    }
   ]
  }
  n25 = LinearInterpolate(A=Multiply(A=@n17.RGB, B=Desaturation(None=@n18.RGB, Fraction=@n17.DeSaturation)), B=@n24, Alpha=@MaterialExpressionNamedRerouteUsage_5)
  n26 = Multiply(A=NamedRerouteUsage(), B=NamedRerouteUsage())
  n27 = ScalarParameter() {defaultValue=1, group=NickelScan_Crystal, parameterName=Emissive_Final_Day, sortPriority=100}
  n28 = Multiply(
   A=Add(A=Add(A=@n25, B=@n26), B=NamedRerouteUsage())
   B=MaterialFunctionCall(
    Value_Morning=@n27
    Value_Day=@n27
    Value_Evening=@n27
    Value_Night=ScalarParameter() {defaultValue=0.5, group=NickelScan_Crystal, parameterName=Emissive_Final_Night, sortPriority=100}
   ) {materialFunction=&Scalar_Time_of_Day}.Result
  )
  NamedRerouteDeclaration(None=ShadingPathSwitch(Default=@n28, Deferred=@n28, Mobile=@n25) {desc=Migrated from FeatureLevelSwitch}) {name=Emissive, nodeColor=<rgba 0.2656 0 1 1>}
  n29 = SetMaterialAttributes!(
   MaterialAttributes
   Base Color=NamedRerouteUsage()
   Metallic=NamedRerouteUsage()
   Specular=NamedRerouteUsage()
   Roughness=NamedRerouteUsage()
   Emissive Color=NamedRerouteUsage()
   Normal=NamedRerouteUsage()
   Opacity=NamedRerouteUsage()
  ) {attributeSetTypes [
    #2
    #3
    #4
    #5
    #1
    #6
    #7
   ]}
  NamedRerouteDeclaration(None=@n29) {name=NickelScan_MA, nodeColor=<rgba 0.4766 1 0.0487 1>}
  n30 = TextureSampleParameter2D() {
   group=NickelScan_Crystal
   parameterName=Crystal_Diffuse
   samplerSource=SSM_Wrap_WorldGroupSettings
   sortPriority=10
   texture=/BRCosmetics/Accessories/FORT_Backpacks/Backpack_M_MED_NickelScan/Textures/T_NickelScan_Backpack_D
  }
  n31 = OneMinus(None=NamedRerouteUsage())
  n32 = ScalarParameter() {defaultValue=1, group=NickelScan_Crystal, parameterName=BaseColor_Intensity}
  NamedRerouteDeclaration(None=Multiply(A=@n30.RGB, B=Multiply(A=@n31, B=@n32))) {name=BaseColor, nodeColor=<rgba 1 0 0.375 1>}
  NamedRerouteDeclaration(None=@n1.M) {name=Mask, nodeColor=<rgba 0 0.3437 1 1>}
  n33 = NamedRerouteDeclaration(None=Subtract(A=NamedRerouteUsage(), B=VectorParameter() {group=NickelScan_Crystal, parameterName=PossitionOffset, defaultValue=<rgba 0 -9.5 -18 1>}.RGB)) {name=OffsetPreskinnedPos, nodeColor=<rgba 1 0 0.9844 1>}
  n34 = Constant()
  n35 = ScalarParameter() {defaultValue=50, group=NickelScan_Crystal, parameterName=Crystal_NoiseRadius}
  NamedRerouteDeclaration(None=Multiply(A=SphereMask(A=@n33, B=@n34, Radius=@n35) {attenuationRadius=15, hardnessPercent=0}, B=NamedRerouteUsage())) {name=NoiseMask, nodeColor=<rgba 0 1 0.6953 1>}
  n36 = Constant() {r=1}
  n37 = DepthFade(FadeDistance=ScalarParameter() {defaultValue=1, group=NickelScan_Crystal, parameterName=Crystal_DepthFadeDist, sortPriority=40})
  NamedRerouteDeclaration(None=MaterialFunctionCall!(Opacity=QualitySwitch(Default=@n36, Low=@n36, High=@n37, Medium=@n36, Epic=@n37)) {materialFunction=/Game/Global/Materials/MaterialGym/Materials/SortedOutline/MF_SortTranslucency}.Result) {name=Opacity, nodeColor=<rgba 0 1 0.1797 1>}
  n38 = TextureSampleParameter2D() {
   group=NickelScan_Crystal
   parameterName=SRM_Map
   samplerSource=SSM_Wrap_WorldGroupSettings
   samplerType=SAMPLERTYPE_Masks
   sortPriority=20
   texture=&T_Fortnite_Default_S
  }
  n39 = NamedRerouteDeclaration(None=@n38.RGB) {name=SpecularMap, nodeColor=<rgba 0 1 0.9766 1>}
  NamedRerouteDeclaration(None=StaticSwitchParameter(
    True=ComponentMask(None=@n39) {r=true}
    False=ScalarParameter() {defaultValue=1, group=NickelScan_Crystal, parameterName=Specular, sortPriority=1}
   ) {group=NickelScan_Crystal, parameterName=-UseTextureMapforSRM, sortPriority=20}) {name=Specular, nodeColor=<rgba 1 0 0.8906 1>}
  MaterialExpressionTransform_0 = Transform(None=CameraVectorWS()) {transformSourceType=TRANSFORMSOURCE_World, transformType=TRANSFORM_Tangent}
  n40 = Convert(None=@MaterialExpressionTransform_0) {
   convertInputs [{
     type=Vector3
     expressionInput {
      expression=&self:MaterialExpressionTransform_0
      inputName=None
      mask=0
      maskA=0
      maskB=0
      maskG=0
      maskR=0
      outputIndex=0
     }
    }]
   convertOutputs [{type=Vector2}, {type=Scalar}]
  }
  n41 = ScalarParameter() {defaultValue=0.1, group=NickelScan_Crystal, parameterName=Crystal_ParallaxDepth, sortPriority=50}
  NamedRerouteDeclaration(None=Divide(A=Multiply(A=@n40, B=@n41), B=Multiply(A=@n40) {constB=-1})) {name=CrystalParallax, nodeColor=<rgba 0.7578 1 0 1>}
  n42 = NamedRerouteUsage()
  NamedRerouteDeclaration(None=StaticSwitchParameter(
    True=ComponentMask(None=@n42) {g=true}
    False=ScalarParameter() {defaultValue=0.2, group=NickelScan_Crystal, parameterName=Roughness, sortPriority=1}
   ) {group=NickelScan_Crystal, parameterName=-UseTextureMapforSRM, sortPriority=20}) {name=Roughness, nodeColor=<rgba 1 0.4922 0 1>}
  n43 = NamedRerouteUsage()
  NamedRerouteDeclaration(None=StaticSwitchParameter(True=ComponentMask(None=@n43) {b=true}, False=ScalarParameter() {group=NickelScan_Crystal, parameterName=Metallic, sortPriority=1}) {group=NickelScan_Crystal, parameterName=-UseTextureMapforSRM, sortPriority=20}) {name=Metallic, nodeColor=<rgba 1 0.4922 0 1>}
  n44 = Power(Base=DotProduct(A=CameraVectorWS(), B=VertexNormalWS())) {constExponent=10}
  n45 = NamedRerouteUsage()
  NamedRerouteDeclaration(None=Multiply(A=Multiply(A=@n44, B=@n45)) {constB=0.1}) {name=CrystalFresnel, nodeColor=<rgba 0.4062 1 0 1>}
  n46 = NamedRerouteUsage()
  n47 = Abs(None=DotProduct(
    A=MaterialFunctionCall() {materialFunction=&CameraDirectionVector}.Result
    B=Transform(None=@n46)
   ))
  n48 = ConstantBiasScale(None=Sine(None=Multiply(A=@n47) {constB=3.14}))
  n49 = ScalarParameter() {defaultValue=8, group=NickelScan_Crystal, parameterName=Cyrstal_Shine_BodyFalloff, sliderMax=1, sortPriority=70}
  n50 = ScalarParameter() {defaultValue=1, group=NickelScan_Crystal, parameterName=Cyrstal_Shine_Intensity, sliderMax=1, sliderMin=-1, sortPriority=70}
  NamedRerouteDeclaration(None=Multiply(A=Power(Base=@n48, Exp=@n49), B=@n50)) {name=Cyrstal Shine, nodeColor=<rgba 1 0 0.7266 1>}
  n51 = NamedRerouteUsage()
  n52 = DotProduct(A=VertexNormalWS(), B=@n51)
  n53 = LinearInterpolate(
   A=Length(None=@n51) {is=class MaterialExpressionLength}
   B=Saturate(None=Multiply(A=@n52) {constB=-1, desc=*-1 invert Dot Product})
   Alpha=Constant() {desc=0.5, r=0.5}
  ) {desc=Distance}
  n54 = ScalarParameter() {defaultValue=0.5, group=NickelScan_Crystal, parameterName=Cyrstal_Glow_SizeExponent, sliderMax=1, sliderMin=0.1, sortPriority=60}
  n55 = Add(A=OneMinus(None=@n54)) {constB=0.1}
  n56 = Saturate(None=Divide(B=Multiply(A=@n53, B=@n55) {desc=*FalloffDamp}))
  n57 = ScalarParameter() {defaultValue=3, group=NickelScan_Crystal, parameterName=Cyrstal_Glow_FalloffExp, sliderMax=10, sortPriority=60}
  n58 = Add(
   A=Multiply(
    A=Power(Base=@n56, Exp=@n57)
    B=ScalarParameter() {defaultValue=1, group=NickelScan_Crystal, parameterName=Cyrstal_Glow_Intensity, sliderMax=1, sortPriority=60}
   )
   B=ScalarParameter() {defaultValue=0.01, group=NickelScan_Crystal, parameterName=Cyrstal_Glow_IntensityMin, sliderMax=1, sortPriority=60}
  )
  n59 = Add(A=@n58, B=Multiply(A=@n58, B=NamedRerouteUsage()))
  n60 = NamedRerouteUsage()
  NamedRerouteDeclaration(None=Saturate(None=Max(A=@n59, B=@n60))) {name=CrystalGlow, nodeColor=<rgba 0.8516 1 0 1>}
  n61 = Divide(A=NamedRerouteUsage(), B=ScalarParameter() {defaultValue=20, group=NickelScan_Crystal, parameterName=Crystal_CrackScale, sortPriority=50})
  n62 = Multiply(A=NamedRerouteUsage()) {constB=0.5}
  n63 = Add(A=ComponentMask(None=@n61) {b=true, r=true}, B=@n62)
  n64 = TextureObject() {samplerType=SAMPLERTYPE_LinearGrayscale, texture=/Game/Global/Textures/Default/Masks_and_Gradients/T_MF_Crystal2}
  n65 = Add(A=ComponentMask(None=@n61) {b=true, g=true}, B=@n62)
  n66 = ComponentMask(None=@n61) {g=true, r=true}
  n67 = LinearInterpolate(
   A=LinearInterpolate(
    A=TextureSample(UVs=@n63, Tex=@n64) {samplerSource=SSM_Wrap_WorldGroupSettings}.R
    B=TextureSample(UVs=@n65, Tex=@n64) {samplerSource=SSM_Wrap_WorldGroupSettings}.R
    Alpha=@n11
   ) {desc=Float3}
   B=TextureSample(UVs=Add(A=@n66, B=@n62), Tex=@n64) {samplerSource=SSM_Wrap_WorldGroupSettings}.R
   Alpha=@n14
  )
  NamedRerouteDeclaration(None=Multiply(A=@n67, B=ScalarParameter() {defaultValue=0.2, group=NickelScan_Crystal, parameterName=Crystal_Distort, sortPriority=50})) {name=CrystalDistort, nodeColor=<rgba 0.9688 0 1 1>}
  NamedRerouteDeclaration(None=VertexInterpolator(VS=LocalPosition() {localOrigin=InstancePreSkinning}.XYZ).PS) {name=PreskinnedPos, nodeColor=<rgba 1 0.2109 0 1>}
  NamedRerouteDeclaration(None=@n67) {name=CrystalMask, nodeColor=<rgba 1 0 0.5859 1>}
  n68 = SkyAtmosphereLightDirection()
  n69 = Transform(None=NamedRerouteUsage())
  n70 = DotProduct(A=@n68, B=@n69)
  n71 = TextureSampleParameter2D(UVs=@n70) {
   group=NickelScan_Crystal
   parameterName=RefractGradient
   samplerSource=SSM_Wrap_WorldGroupSettings
   sortPriority=90
   texture=/Game/Characters/Player/Female/Medium/Bodies/F_MED_Critter_Streak/Textures/CritterStreak_Gradient
  }
  n72 = ScalarParameter() {defaultValue=5, group=NickelScan_Crystal, parameterName=SdN_Pow, sortPriority=90}
  n73 = Multiply(
   A=@n71.RGB
   B=Multiply(A=Power(Base=@n70, Exp=@n72), B=ScalarParameter() {defaultValue=10, group=NickelScan_Crystal, parameterName=SdN_Mult, sortPriority=90}) {constB=0.5}
  )
  n74 = ScalarParameter() {defaultValue=2, group=NickelScan_Crystal, parameterName=Refraction IOR, sortPriority=90}
  n75 = TextureSampleParameter2D(UVs=MaterialFunctionCall!(Vector=MaterialFunctionCall!(Ray Direction=@n68, Surface Normal=@n69, Refractive Index Origin, Refractive Index Target=@n74) {materialFunction=/Engine/Functions/Engine_MaterialFunctions01/Vectors/Refract}.Vector) {materialFunction=/Engine/Functions/Engine_MaterialFunctions01/Coordinates/LongLatToUV}.Result) {
   group=NickelScan_Crystal
   parameterName=RefractGradient
   samplerSource=SSM_Wrap_WorldGroupSettings
   sortPriority=90
   texture=&CritterStreak_Gradient
  }
  NamedRerouteDeclaration(None=LinearInterpolate(
    A=ScalarParameter() {group=NickelScan_Crystal, parameterName=FakeHighLight_Low, sortPriority=90}
    B=ScalarParameter() {defaultValue=0.3, group=NickelScan_Crystal, parameterName=FakeHighLight_High, sortPriority=90}
    Alpha=Max(A=@n73, B=@n75.RGB)
   )) {name=FakeLighting, nodeColor=<rgba 1 0 0 1>}
  NamedRerouteDeclaration(None=@n1.Fresnel) {name=Fresnel, nodeColor=<rgba 0 1 0.4375 1>}
  n76 = TextureSampleParameter2D(UVs=Multiply(A=TextureCoordinate(), B=ScalarParameter() {defaultValue=200, group=NickelScan_Crystal, parameterName=MicroNormalTiling, sortPriority=30}) {constB=200}) {
   group=NickelScan_Crystal
   parameterName=MicroNormal
   samplerType=SAMPLERTYPE_Normal
   sortPriority=31
   texture=/BRCosmetics/Global/Textures/Default/Tileable_Patterns/T_PlexiGlass_MicroNormal
  }
  n77 = TextureSampleParameter2D() {
   group=NickelScan_Crystal
   parameterName=Normal
   samplerSource=SSM_Wrap_WorldGroupSettings
   samplerType=SAMPLERTYPE_Normal
   sortPriority=30
   texture=&FlatNormalBlue
  }
  NamedRerouteDeclaration(None=MaterialFunctionCall(BaseNormal=@n76.RGB, AdditionalNormal=NamedRerouteDeclaration(None=@n77.RGB) {name=Normal_In, nodeColor=<rgba 0.4297 0 1 1>}) {materialFunction=&BlendAngleCorrectedNormals}.Result) {name=Normal_Out, nodeColor=<rgba 0 1 0.1094 1>}
  n78 = NamedRerouteUsage()
  n79 = MaterialFunctionCall!(Circumference Tiling, Position=LocalPosition().XYZ, Z Height, HQ Angle Calculation?) {materialFunction=/Game/Characters/Player/Common/MaterialFunctions/MF_CylindricalUVs_New}
  n80 = Multiply(
   A=Constant2Vector() {g=2, r=1}
   B=ScalarParameter() {defaultValue=1, group=NickelScan_Crystal, parameterName=ThinFilmUVTiling, sortPriority=81}
  )
  n81 = MaterialFunctionCall(
   Normal=TextureSample(UVs=Multiply(A=@n79.UVs, B=@n80)) {samplerType=SAMPLERTYPE_Normal, texture=/Game/Gadgets/Assets/Telephone_Booth/Textures/T_BlindsNormalNoise_N}.RGB
   Flatness=ScalarParameter() {defaultValue=0.75, group=NickelScan_Crystal, parameterName=ThinFilmNormalFlatten, sortPriority=81}
  ) {materialFunction=&FlattenNormal}
  n82 = Fresnel(
   ExponentIn=ScalarParameter() {defaultValue=2, group=NickelScan_Crystal, parameterName=ThinFilmExp, sortPriority=81}
   Normal=Transform(None=MaterialFunctionCall(BaseNormal=@n78, AdditionalNormal=@n81.Result) {materialFunction=&BlendAngleCorrectedNormals}.Result)
  ) {baseReflectFraction=0.1, exponent=3}
  n83 = TextureSampleParameter2D(UVs=@n82) {
   group=NickelScan_Crystal
   parameterName=ThinFilmTexture
   sortPriority=80
   texture=/Game/Global/Textures/Default/Thinfilm/T_ThinFilm_Spectrum_CoolColor
  }
  n84 = Multiply(A=@n83.RGB, B=@n82)
  n85 = ScalarParameter() {defaultValue=0.6, group=NickelScan_Crystal, parameterName=ThinFilmDesaturation, sortPriority=81}
  NamedRerouteDeclaration(None=Multiply(
    A=Desaturation(None=@n84, Fraction=@n85)
    B=ScalarParameter() {defaultValue=1, group=NickelScan_Crystal, parameterName=ThinFilmMult, sortPriority=81}
   )) {name=ThinFilm, nodeColor=<rgba 0.7812 0 1 1>}
  NamedRerouteUsage()
  VertexNormalWS()
 ]
}
