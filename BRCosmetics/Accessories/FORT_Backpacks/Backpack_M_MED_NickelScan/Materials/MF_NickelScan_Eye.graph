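# Material function for the NickelScan backpack's animated eye, outputting an emissive color and a mask.
# Sine-based eyelids, layered pupil shapes and glow build the eye, using bump offset, view parallax and noise distortion for depth.
# Custom HLSL adds hashed single or double blinks and a shake, alongside procedural pupil darting.
# A CurrentStage timeline, timed by GameTimeFromBP, remaps into three reaction stages with eye and pupil offsets.
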MaterialFunction /BRCosmetics/Accessories/FORT_Backpacks/Backpack_M_MED_NickelScan/Materials/MF_NickelScan_Eye
graph {nodes [
  AtmosphericLightVector()
  FunctionOutput(None=NamedRerouteUsage()) {outputName=EmssiveColor}
  n1 = Power(Base=DotProduct(A=CameraVectorWS(), B=NamedRerouteUsage())) {constExponent=8}
  n2 = NamedRerouteUsage()
  n3 = Constant() {r=0.5}
  n4 = Multiply(A=Power(Base=SphereMask(A=@n2, B=@n3) {attenuationRadius=0.35, hardnessPercent=0}), B=NamedRerouteUsage())
  n5 = Multiply(A=Multiply(A=Multiply(A=@n1, B=@n4)) {constB=200}, B=NamedRerouteUsage())
  n6 = Max(A=@n5, B=NamedRerouteUsage())
  n7 = NamedRerouteUsage()
  n8 = TextureCoordinate()
  n9 = Min(A=OneMinus(None=@n8), B=@n8)
  n10 = Multiply(A=ComponentMask(None=@n9) {r=true}, B=ComponentMask(None=@n9) {g=true})
  FunctionOutput(None=Multiply(A=Max(A=@n6, B=@n7), B=SmoothStep(Value=@n10) {constMax=0.05, desc=SafeMask})) {outputName=Mask}
  n11 = Time()
  MaterialExpressionTime_1 = Time()
  MaterialExpressionConstant_13 = Constant() {r=0.25}
  MaterialExpressionConstant_14 = Constant() {r=1.5}
  n12 = Custom!(Time=@MaterialExpressionTime_1, Freq=@MaterialExpressionConstant_13, Amp=@MaterialExpressionConstant_14) {
   code="// Inputs: Time, Freq, Amp\r\nfloat timeSlot = floor(Time * Freq);\r\nfloat phase = frac(Time * Freq);\r\n\r\n// Simple pseudo-random hash function\r\nfloat rand = frac(sin(timeSlot * 12.9898) * 43758.5453);\r\n\r\nfloat blink = 0.0;\r\nfloat duration = 0.3; // Total window for blinking (30% of the cycle)\r\n\r\n// Logic: \r\n// If rand < 0.4: Stay open (No blink)\r\n// If rand 0.4 - 0.8: Single Blink\r\n// If rand > 0.8: Double Blink\r\n\r\nif (rand > 0.4 &&&& phase > (1.0 - duration)) {\r\n    // Normalize phase to 0.0 - 1.0 within the blink window\r\n    float t = (phase - (1.0 - duration)) / duration;\r\n\r\n    if (rand > 0.8) {\r\n        // --- DOUBLE BLINK LOGIC ---\r\n        // Two smaller peaks within the 't' window\r\n        blink = saturate(sin(t * 3.14159 * 2.0)); \r\n        blink = pow(blink, 2.0); // Sharpen the curve\r\n    } else {\r\n        // --- SINGLE BLINK LOGIC ---\r\n        // Asymmetric: Fast close (30%), Slower open (70%)\r\n        if (t < 0.3) {\r\n            blink = pow(t / 0.3, 2.0);\r\n        } else {\r\n            blink = 1.0 - pow((t - 0.3) / 0.7, 1.5);\r\n        }\r\n    }\r\n}\r\n\r\nreturn saturate(blink * Amp);"
   description=EyeOpenBehavior
   inputs [
    {
     inputName=Time
     input {expression=&self:MaterialExpressionTime_1}
    }
    {
     inputName=Freq
     input {expression=&self:MaterialExpressionConstant_13}
    }
    {
     inputName=Amp
     input {expression=&self:MaterialExpressionConstant_14}
    }
   ]
  }
  n13 = OneMinus(None=@n12)
  Max(A=Sine(None=Multiply(A=@n11) {constB=0.15}), B=@n13)
  n14 = Constant2Vector() {r=-0.05}
  n15 = Add(A=NamedRerouteUsage(), B=@n14) {constB=0}
  n16 = ComponentMask(None=@n15) {r=true}
  n17 = Constant() {r=0.75}
  n18 = Subtract(A=Multiply(A=@n16, B=@n17) {constB=3.5}) {constB=0.1}
  n19 = NamedRerouteUsage()
  n20 = Add(A=Multiply(A=Sine(None=@n18), B=@n19)) {constB=0.5}
  n21 = ComponentMask(None=@n15) {g=true}
  n22 = Multiply(A=Multiply(A=@n16) {constB=-1}, B=@n17) {constB=3.5}
  n23 = Multiply(A=Sine(None=Add(A=@n22) {constB=0.1}), B=@n19)
  n24 = Max(A=Add(A=Multiply(A=@n20) {constB=-1}, B=@n21), B=Subtract(A=Add(A=@n23) {constB=0.5}, B=@n21))
  NamedRerouteDeclaration(None=Step(Y=@n24) {constX=0}) {name=Graph_Eye_Msk, nodeColor=<rgba 0.4062 0 1 1>}
  n25 = TextureCoordinate()
  n26 = NamedRerouteUsage()
  n27 = NamedRerouteUsage()
  n28 = NamedRerouteUsage()
  n29 = NamedRerouteUsage()
  n30 = Max(A=Max(A=Multiply(A=@n26, B=@n27), B=Multiply(A=@n28, B=@n29)), B=Multiply(A=NamedRerouteUsage(), B=NamedRerouteUsage()))
  NamedRerouteDeclaration(None=MaterialFunctionCall(UVs=Add(A=@n25, B=@n30), "Rotation Angle (0-1)"=Constant() {r=0.05}) {materialFunction=&CustomRotator}.Rotated Values) {name=BaseUV, nodeColor=<rgba 0 1 0.0859 1>}
  n31 = TextureSample(UVs=NamedRerouteUsage()) {
   samplerSource=SSM_Wrap_WorldGroupSettings
   samplerType=SAMPLERTYPE_Normal
   texture=/BRCosmetics/Accessories/FORT_Backpacks/Backpack_M_MED_NickelScan/Textures/T_NickelScan_EyeBall_N
  }
  NamedRerouteDeclaration(None=Transform(None=@n31.RGB)) {name=EyeNormal_Distort, nodeColor=<rgba 0 1 0.9062 1>}
  n32 = Constant() {r=0.5}
  n33 = ScalarParameter() {defaultValue=0.02, group=NickelScan_Eye, parameterName=EyeDartingTransition}
  n34 = Add(A=@n32, B=@n33)
  n35 = Subtract(A=@n32, B=@n33)
  n36 = TextureCoordinate() {uTiling=0, vTiling=0}
  n37 = Multiply(A=ScalarParameter() {defaultValue=1, group=NickelScan_Eye, parameterName=JitterAmount}, B=Time())
  n38 = Panner(Coordinate=@n36, Time=@n37) {speedX=-0.1, speedY=0.2}
  n39 = Subtract(A=Multiply(A=Frac(None=@n38)) {constB=2})
  n40 = Saturate(None=SmoothStep(Min=@n34, Max=@n35, Value=Abs(None=@n39)) {constMax=0.5, constMin=0.55})
  n41 = Multiply(A=ComponentMask(None=@n40) {r=true}) {constB=0.2}
  n42 = Multiply(A=ComponentMask(None=@n40) {g=true}) {constB=0.4}
  n43 = Panner(Coordinate=@n36, Time=@n37) {speedX=0.12, speedY=-0.05}
  n44 = Subtract(A=Multiply(A=Frac(None=@n43)) {constB=2})
  n45 = Saturate(None=SmoothStep(Min=@n34, Max=@n35, Value=Abs(None=@n44)) {constMax=0.5, constMin=0.55})
  n46 = Multiply(A=ComponentMask(None=@n45) {r=true}) {constB=0.2}
  n47 = Multiply(A=ComponentMask(None=@n45) {g=true}) {constB=0.4}
  n48 = Multiply(A=AppendVector(A=Add(A=@n41, B=@n42), B=Add(A=@n46, B=@n47)))
  n49 = LinearInterpolate(A=ConstantBiasScale(None=Clamp(None=@n48)) {bias=-0.5, scale=0.15}, B=NamedRerouteUsage(), Alpha=NamedRerouteUsage()) {constAlpha=1, constB=0}
  n50 = NamedRerouteUsage()
  n51 = NamedRerouteUsage()
  n52 = NamedRerouteUsage()
  NamedRerouteDeclaration(None=LinearInterpolate(A=LinearInterpolate(A=@n49, B=@n50, Alpha=@n51) {constAlpha=1, constB=0}, B=Multiply(A=@n52) {constB=-0.5}, Alpha=NamedRerouteUsage()) {constAlpha=1, constB=0}) {name=PupilJitterOffst, nodeColor=<rgba 0.7102 0.2029 1 1>}
  n53 = NamedRerouteUsage()
  n54 = ScalarParameter() {defaultValue=0.1, group=NickelScan_Eye, parameterName=EyeDistortion}
  n55 = Multiply(
   A=TextureSample(UVs=Panner(Coordinate=@n53) {speedX=0.02, speedY=0.08}) {samplerSource=SSM_Wrap_WorldGroupSettings, texture=/Game/Global/Textures/Default/Noise/T_FireNoise_01}.R
   B=Divide(A=@n54)
  ) {constB=0.04}
  n56 = Add(A=@n53, B=@n55)
  n57 = Multiply(A=ComponentMask(None=@n56) {r=true}) {constB=40}
  n58 = Abs(None=ConstantBiasScale(None=Frac(None=@n57)) {bias=-0.5, scale=2})
  n59 = Step(Y=Frac(None=ConstantBiasScale(None=@n57) {bias=0.5})) {constX=0.5}
  n60 = Frac(None=ConstantBiasScale(None=@n57) {bias=0.5, scale=0.25})
  n61 = ComponentMask(None=@n56) {g=true}
  n62 = Multiply(A=Multiply(A=Multiply(A=@n58, B=@n59), B=Step(Y=@n60) {constX=0.5}), B=Abs(None=ConstantBiasScale(None=@n61) {bias=-0.5, scale=2}))
  n63 = LinearInterpolate(A=@n54, B=Multiply(A=@n54) {constB=-1}, Alpha=Step(Y=ComponentMask(None=@n53) {g=true}, X=Constant() {r=0.5}) {constX=0.5})
  NamedRerouteDeclaration(None=Add(A=Multiply(A=@n62, B=@n63) {constB=0.04}, B=Multiply(A=@n55) {constB=0.25})) {name=Distort_Eyelid, nodeColor=<rgba 1 0 0.1172 1>}
  NamedRerouteDeclaration(None=@n55) {name=Distort_Pupil, nodeColor=<rgba 1 0 0.1172 1>}
  MaterialExpressionTransform_2 = Transform(None=CameraVectorWS()) {transformSourceType=TRANSFORMSOURCE_World, transformType=TRANSFORM_Tangent}
  n64 = Convert(None=@MaterialExpressionTransform_2) {
   convertInputs [{
     type=Vector3
     expressionInput {
      expression=&self:MaterialExpressionTransform_2
      inputName=None
      mask=0
      maskA=0
      maskB=0
      maskG=0
      maskR=0
      outputIndex=0
     }
    }]
   convertOutputs [{type=Vector2}, {type=Scalar}]
  }
  n65 = Constant() {r=0.2}
  NamedRerouteDeclaration(None=Divide(A=Multiply(A=@n64, B=@n65), B=Multiply(A=@n64) {constB=-1})) {name=EyeParallax, nodeColor=<rgba 0.7578 1 0 1>}
  n66 = NamedRerouteUsage()
  n67 = NamedRerouteUsage()
  n68 = Add(A=Add(A=Add(A=@n66, B=@n67), B=NamedRerouteUsage()), B=NamedRerouteUsage()) {constB=0}
  n69 = Add(A=@n68, B=Constant2Vector() {r=-0.005})
  n70 = Constant2Vector() {g=0.5, r=0.5}
  NamedRerouteDeclaration(None=Step(Y=Distance(A=@n69, B=@n70)) {constX=0.09}) {name=PupilShape_Center, nodeColor=<rgba 0.5234 1 0 1>}
  n71 = Add(A=@n68, B=Constant2Vector() {r=-0.005})
  n72 = Constant2Vector() {g=0.5, r=0.5}
  NamedRerouteDeclaration(None=Step(Y=Distance(A=@n71, B=@n72)) {constX=0.16}) {name=PupilShape_Base, nodeColor=<rgba 0.5234 1 0 1>}
  n73 = TextureSample(UVs=NamedRerouteUsage()) {
   samplerSource=SSM_Wrap_WorldGroupSettings
   samplerType=SAMPLERTYPE_Normal
   texture=&T_NickelScan_EyeBall_N
  }
  NamedRerouteDeclaration(None=Transform(None=@n73.RGB)) {name=EyeNormal, nodeColor=<rgba 0 1 0.9063 1>}
  n74 = NamedRerouteUsage()
  n75 = BumpOffset(Coordinate=@n74, Height=SphereMask(A=@n74, B=Constant() {r=0.5}) {attenuationRadius=0.5, hardnessPercent=0}) {heightRatio=0.1}
  NamedRerouteDeclaration(None=Add(A=@n75, B=NamedRerouteUsage())) {name=EyeUV_Distorted, nodeColor=<rgba 0 0.6016 1 1>}
  n76 = NamedRerouteUsage()
  n77 = Add(A=@n76) {constB=-0.5}
  n78 = NamedRerouteUsage()
  n79 = SphereMask(
   A=Add(A=Add(A=@n77, B=@n78), B=NamedRerouteUsage())
   B=Constant()
   Radius=ScalarParameter() {defaultValue=0.2, group=NickelScan_Eye, parameterName=PupilGlowRadius}
  ) {attenuationRadius=0.2, hardnessPercent=50}
  NamedRerouteDeclaration(None=Multiply(A=Power(Base=@n79) {constExponent=3}, B=ScalarParameter() {group=NickelScan_Eye, parameterName=PupilGlow_Intensity}) {constB=0}) {name=PupilGlow, nodeColor=<rgba 1 0 0.3281 1>}
  n80 = Saturate(None=Subtract(B=@n24) {constA=0})
  n81 = Step(X=@n80) {constX=0.1, constY=0.015}
  n82 = Subtract(A=Step(X=@n80) {constX=0.1}, B=@n81)
  n83 = Step(Y=ComponentMask(None=NamedRerouteUsage()) {g=true}) {constX=0.5}
  NamedRerouteDeclaration(None=Multiply(A=@n82, B=@n83) {desc=Graph_Eyelid}) {name=Graph_Eyelid_Upper_Msk, nodeColor=<rgba 0.4062 0 1 1>}
  n84 = Add(A=ComponentMask(None=@n76) {g=true}) {constB=-0.25}
  NamedRerouteDeclaration(None=Saturate(None=Multiply(A=@n84) {constB=2})) {name=EyeShadow, nodeColor=<rgba 0.6641 1 0 1>}
  NamedRerouteDeclaration(None=@n81) {name=Graph_Eye_Inner_Msk, nodeColor=<rgba 0.4062 0 1 1>}
  n85 = Multiply(
   A=LinearInterpolate(
    A=VectorParameter() {group=NickelScan_Eye, parameterName=EyelidColor_Lower, defaultValue=<rgba 0.1 0.1 0.1 1>}.RGB
    B=VectorParameter() {group=NickelScan_Eye, parameterName=EyelidColor_Upper, defaultValue=<rgba 0.8 0.8 0.8 1>}.RGB
    Alpha=NamedRerouteUsage()
   )
   B=@n5
  )
  n86 = VectorParameter() {group=NickelScan_Eye, parameterName=EyeColor_Main, defaultValue=<rgba 0.2423 0.6724 0.0123 1>}
  n87 = NamedRerouteUsage()
  n88 = Multiply(A=NamedRerouteUsage(), B=@n87) {constB=0.5}
  n89 = VectorParameter() {group=NickelScan_Eye, parameterName=EyeColor_Sub, defaultValue=<rgba 0.4397 0.9047 0.0666 1>}
  n90 = Multiply(A=@n89.RGB, B=Multiply(A=NamedRerouteUsage(), B=@n87)) {constB=10}
  n91 = NamedRerouteUsage()
  n92 = VectorParameter() {group=NickelScan_Eye, parameterName=EyeColor_CenterColor, defaultValue=<rgba 0.2423 0.6724 0.0123 1>}
  n93 = NamedRerouteUsage()
  n94 = LinearInterpolate(A=@n86.RGB, B=LinearInterpolate(A=Multiply(A=@n89.RGB, B=@n91), B=Multiply(A=@n92.RGB, B=@n93), Alpha=@n93), Alpha=@n91)
  n95 = NamedRerouteUsage()
  n96 = NamedRerouteUsage()
  n97 = LinearInterpolate(
   A=Max(A=Multiply(A=@n86.RGB, B=@n88) {constB=0.5}, B=Multiply(A=@n90) {constB=10})
   B=Add(A=Multiply(A=@n94, B=@n95), B=Multiply(A=@n96, B=@n92.RGB))
   Alpha=NamedRerouteUsage()
  )
  n98 = NamedRerouteUsage()
  n99 = NamedRerouteUsage()
  NamedRerouteDeclaration(None=Add(A=LinearInterpolate(A=@n85, B=@n97, Alpha=@n98), B=Multiply(A=@n99, B=@n89.RGB))) {name=Eye_Colored, nodeColor=<rgba 0 0.0156 1 1>}
  n100 = NamedRerouteUsage()
  NamedRerouteDeclaration(None=SmoothStep(Value=ComponentMask(None=@n100) {g=true}) {constMax=0, constMin=0.5}) {name=HalfMask, nodeColor=<rgba 0 0.5312 1 1>}
  NamedRerouteDeclaration(None=@n75) {name=EyeUV, nodeColor=<rgba 0 0.6016 1 1>}
  n101 = Add(A=NamedRerouteUsage(), B=@n14) {constB=0}
  n102 = ComponentMask(None=@n101) {r=true}
  n103 = Constant() {r=0.75}
  n104 = Sine(None=Subtract(A=Multiply(A=@n102, B=@n103) {constB=3.5}) {constB=0.1})
  n105 = NamedRerouteUsage()
  n106 = Multiply(A=Add(A=Multiply(A=@n104, B=@n105)) {constB=0.5}) {constB=-1}
  n107 = ComponentMask(None=@n101) {g=true}
  n108 = Add(A=Multiply(A=Multiply(A=@n102) {constB=-1}, B=@n103) {constB=3.5}) {constB=0.1}
  n109 = Add(A=Multiply(A=Sine(None=@n108), B=@n105)) {constB=0.5}
  NamedRerouteDeclaration(None=Max(A=Add(A=@n106, B=@n107), B=Subtract(A=@n109, B=@n107))) {name=Eyelid_HightMsk, nodeColor=<rgba 0 1 0.4141 1>}
  n110 = Subtract(A=Time(), B=ScalarParameter() {defaultValue=-99999, group=NickelScan_Eye, parameterName=GameTimeFromBP, sortPriority=100})
  n111 = Add(A=ScalarParameter() {defaultValue=1, group=NickelScan_Eye, parameterName=Animation_Duration, sortPriority=101}) {constB=0}
  NamedRerouteDeclaration(None=Saturate(None=Divide(A=@n110, B=@n111))) {name=TimeNative, nodeColor=<rgba 0.2891 1 0 1>}
  NamedRerouteDeclaration(None=@n89.RGB) {name=EyeColor_Sub, nodeColor=<rgba 1 0.8672 0 1>}
  n112 = NamedRerouteUsage()
  n113 = Constant4Vector() {desc=Stage 0 to 1, constant=<rgba 0 1 0 1>}
  n114 = Saturate(None=MaterialFunctionCall(Input=@n112, Input Low=@n113.R, Input High=@n113.R, Target Low=@n113.R, Target High=@n113.R) {materialFunction=&RemapValueRange}.Result)
  NamedRerouteDeclaration(None=ComponentMask(None=ColorRamp!(None=@n114) {colorCurve="<self>.ColorCurve"}) {r=true}) {name=Stage1, nodeColor=<rgba 0.2656 0 1 1>}
  n115 = Frac(None=Multiply(A=@n114) {constB=1.5})
  n116 = ComponentMask(None=Power(Base=ColorRamp(None=@n115) {colorCurve="<self>.ColorCurve"})) {r=true}
  NamedRerouteDeclaration(None=LinearInterpolate(A=Constant2Vector() {g=-0.05, r=0.15}, B=Constant2Vector() {g=-0.05, r=-0.25}, Alpha=@n116)) {name=Stage1_PupilOffset, nodeColor=<rgba 0.2422 0 1 1>}
  NamedRerouteDeclaration(None=Multiply(A=@n82, B=OneMinus(None=@n83)) {desc=Graph_Eyelid}) {name=Graph_Eyelid_lower_Msk, nodeColor=<rgba 0.4062 0 1 1>}
  NamedRerouteDeclaration(None=LinearInterpolate(A=Constant2Vector() {r=0.05}, B=Constant2Vector() {r=-0.075}, Alpha=@n116)) {name=Stage1_EyeOffset, nodeColor=<rgba 0.2422 0 1 1>}
  n117 = Constant4Vector() {desc=Stage 1 to 2, constant=<rgba 1 2 0 1>}
  n118 = Saturate(None=MaterialFunctionCall(Input=@n112, Input Low=@n117.R, Input High=@n117.R, Target Low=@n117.R, Target High=@n117.R) {materialFunction=&RemapValueRange}.Result)
  NamedRerouteDeclaration(None=ComponentMask(None=ColorRamp(None=@n118) {colorCurve="<self>.ColorCurve"}) {r=true}) {name=Stage2, nodeColor=<rgba 0.2656 0 1 1>}
  NamedRerouteDeclaration(None=LinearInterpolate(A=Constant2Vector() {g=0.2, r=-0.05}, B=Constant2Vector() {g=-0.3, r=-0.05}, Alpha=@n118)) {name=Stage2_PupilOffset, nodeColor=<rgba 0.2422 0 1 1>}
  NamedRerouteDeclaration(None=LinearInterpolate(A=Constant2Vector() {g=0.15}, B=Constant2Vector() {g=-0.2}, Alpha=@n118)) {name=Stage2_EyeOffset, nodeColor=<rgba 0.2422 0 1 1>}
  MaterialExpressionTime_2 = Time()
  MaterialExpressionConstant_15 = Constant() {r=40}
  MaterialExpressionConstant_16 = Constant() {r=0.05}
  n119 = Custom!(Time=@MaterialExpressionTime_2, Speed=@MaterialExpressionConstant_15, Amount=@MaterialExpressionConstant_16) {
   code="float2 smoothShake;\r\nsmoothShake.x = sin(Time * Speed) * cos(Time * Speed * 0.8);\r\nsmoothShake.y = cos(Time * Speed * 1.1) * sin(Time * Speed * 0.7);\r\nreturn smoothShake * Amount;"
   description=RandomOffset
   outputType=CMOT_Float2
   inputs [
    {
     inputName=Time
     input {expression=&self:MaterialExpressionTime_2}
    }
    {
     inputName=Speed
     input {expression=&self:MaterialExpressionConstant_15}
    }
    {
     inputName=Amount
     input {expression=&self:MaterialExpressionConstant_16}
    }
   ]
  }
  NamedRerouteDeclaration(None=Add(A=@n119) {constB=0.04}) {name=Stage3_EyeOffset, nodeColor=<rgba 0.2422 0 1 1>}
  n120 = Constant4Vector() {desc=Stage 2 to 3, constant=<rgba 2 3 0 1>}
  n121 = Saturate(None=MaterialFunctionCall(Input=@n112, Input Low=@n120.R, Input High=@n120.R, Target Low=@n120.R, Target High=@n120.R) {materialFunction=&RemapValueRange}.Result)
  NamedRerouteDeclaration(None=ComponentMask(None=ColorRamp(None=@n121) {colorCurve="<self>.ColorCurve"}) {r=true}) {name=Stage3, nodeColor=<rgba 0.2656 0 1 1>}
  n122 = ScalarParameter() {group=NickelScan_Eye, parameterName=CurrentStage, sortPriority=3}
  n123 = LinearInterpolate(A=Subtract(A=@n122), B=@n122, Alpha=NamedRerouteUsage()) {desc=Native lerp toward}
  n124 = Constant() {r=3}
  NamedRerouteDeclaration(None=Multiply(A=Fmod(A=@n123, B=@n124), B=ScalarParameter() {defaultValue=1, group=NickelScan_Eye, parameterName=ReactivityOn})) {name=TimeLine, nodeColor=<rgba 0.0312 0 1 1>}
  NamedRerouteDeclaration(None=@n118) {name=Stage2_Transition, nodeColor=<rgba 0.2656 0 1 1>}
  n125 = DotProduct(A=Constant3Vector() {constant=<rgba 0 0 1 0>}, B=NamedRerouteUsage())
  n126 = Subtract(A=NamedRerouteUsage()) {constB=0.5}
  n127 = Constant()
  n128 = ScalarParameter() {group=NickelScan_Eye, parameterName=SphericalFresnel_Hardness, sortPriority=10}
  NamedRerouteDeclaration(None=Multiply(A=Abs(None=@n125), B=SphereMask(A=@n126, B=@n127, Hardness=@n128) {attenuationRadius=0.45, hardnessPercent=1}) {constB=0.1}) {name=CombinedMask, nodeColor=<rgba 0 0.5312 1 1>}
  n129 = MaterialFunctionCall(UVs=NamedRerouteUsage(), "Rotation Angle (0-1)"=Constant() {r=0.12}) {materialFunction=&CustomRotator}
  n130 = Subtract(A=Multiply(A=ComponentMask(None=@n129.Rotated Values) {r=true}) {constB=2})
  n131 = Multiply(
   A=SphereMask(A=Subtract(A=@n129.Rotated Values) {constB=0.5}, B=Constant()) {attenuationRadius=1, hardnessPercent=0}
   B=InverseLinearInterpolate(Value=Abs(None=@n130)) {constA=0.125, constB=0}
  )
  n132 = Panner(Coordinate=Multiply(A=@n129.Rotated Values, B=Constant2Vector() {g=0.05, r=2})) {speedX=0.05, speedY=0.1}
  n133 = NamedRerouteUsage()
  n134 = Multiply(
   A=Multiply(
    A=Saturate(None=@n131)
    B=TextureSample(UVs=@n132) {samplerSource=SSM_Wrap_WorldGroupSettings, texture=&T_FireNoise_01}.R
   )
   B=Saturate(None=Subtract(B=@n133))
  )
  NamedRerouteDeclaration(None=Multiply(A=Multiply(A=@n134) {constB=10}, B=NamedRerouteUsage())) {name=EyeScarBeam, nodeColor=<rgba 0 1 0.3672 1>}
  n135 = LinearInterpolate(A=@n13, Alpha=NamedRerouteUsage()) {constB=1.1}
  n136 = NamedRerouteUsage()
  n137 = LinearInterpolate(
   A=LinearInterpolate(Alpha=@n136) {constA=0.5, constB=1.5}
   B=LinearInterpolate(Alpha=@n136) {constA=1.5, constB=0.5}
   Alpha=NamedRerouteUsage()
  )
  n138 = NamedRerouteUsage()
  n139 = Multiply(A=LinearInterpolate(A=LinearInterpolate(A=@n135, B=@n137, Alpha=@n138), Alpha=NamedRerouteUsage()) {constB=1.4}) {constB=0.15}
  NamedRerouteDeclaration(None=Frac(None=Abs(None=@n139))) {name=eyeAmp, nodeColor=<rgba 0 0.4609 1 1>}
  NamedRerouteUsage()
 ]}
