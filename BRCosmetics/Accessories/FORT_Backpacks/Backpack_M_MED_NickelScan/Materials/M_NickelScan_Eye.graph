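# Translucent unlit material for the glowing eye sprite on the NickelScan backpack cosmetic, built for skeletal meshes.
# MF_NickelScan_Eye's mask drives both custom glow graph and color in the shared MF_GlowSprites function.
# World Position Offset expands sprites along negative local X and Z using centered UVs, scaled by SpritesSize.
# The FP_Mode switch, on by default, adds first-person camera offset and FOV interpolation functions.
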Material /BRCosmetics/Accessories/FORT_Backpacks/Backpack_M_MED_NickelScan/Materials/M_NickelScan_Eye
graph {
 properties {
  bAutomaticallySetUsageInEditor=false
  bCastRayTracedShadows=false
  bUseMaterialAttributes=true
  bUsedWithSkeletalMesh=true
  blendMode=BLEND_Translucent
  shadingModel=MSM_Unlit
  usedShadingModels=Unlit
 }
 nodes [
  Constant() {r=1}
  MaterialFunctionCall() {materialFunction=&MF_TechArtOverrides}
  n1 = MaterialFunctionCall() {materialFunction=&/MF_NickelScan_Eye}
  n2 = MaterialFunctionCall(Float2=Subtract(A=TextureCoordinate()) {constB=0.5}) {materialFunction=&BreakOutFloat2Components}
  n3 = Multiply(A=Transform(None=Constant3Vector() {constant=<rgba -1 0 0 0>}) {transformSourceType=TRANSFORMSOURCE_Local}, B=@n2.R)
  n4 = Multiply(A=Transform(None=Constant3Vector() {constant=<rgba 0 0 -1 0>}) {transformSourceType=TRANSFORMSOURCE_Local}, B=@n2.G)
  n5 = SetMaterialAttributes!(
   MaterialAttributes=MaterialFunctionCall!(UseCustomGlowGraph=StaticBool() {value=true}, CustomGlowGraph=@n1.Mask, CustomGlowColor=@n1.Mask) {materialFunction=/Game/Global/Materials/Character_Parent_2023/GlowSprites_Parent_2026/MF_GlowSprites}.MA
   World Position Offset=Multiply(A=Add(A=@n3, B=@n4), B=ScalarParameter() {defaultValue=1, group=GlowSprites 05 Size, parameterName=SpritesSize})
  ) {attributeSetTypes [#1]}
  StaticSwitchParameter(
   True=MaterialFunctionCall!(In=MaterialFunctionCall!(In=@n5) {materialFunction=/Game/Packages/Fortress_SharedMaterials/Base_Material_Functions/BaseCharacterShaderFunctions/MF_FPMeshCameraOffset}.Result) {materialFunction=/Game/Packages/Fortress_SharedMaterials/Base_Material_Functions/BaseCharacterShaderFunctions/MF_InterpFP_FOV}.Result
   False=@n5
  ) {defaultValue=true, group=FP Mode, parameterName=FP_Mode, sortPriority=10}
 ]
}
