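# Animation Blueprint for the NickelScan backpack, which tilts the pack and animates a mechanical eye with eyelids.
# The event graph derives pack rotation from the owner's spine yaw, using a scaled, rolled variant while skydiving.
# It also toggles a randomly timed blink value and, for two seconds after an enemy elimination, computes one of three sine/cosine eye motions.
# The AnimGraph chains additive pack and eyelid bone transforms into a Look At toward an eye target bone.
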AnimBlueprint /BRCosmetics/Accessories/FORT_Backpacks/Backpack_M_MED_NickelScan/Meshes/M_MED_NickelScan_Pack_AnimBP
target_skeleton=&/M_MED_NickelScan_Pack_Skeleton
graphs [
 AnimGraph {nodes [
   n1 = "Transform (Modify) Bone - Bone: pack"!(ComponentPose=Local To Component(LocalPose=Input Pose()), bAlphaBoolEnabled:bool, Translation="0,2,0", Rotation="0,0,0", Scale) -> (Pose) {
    alphaInputType=Bool
    rotationMode=BMM_Additive
    rotationSpace=BCS_BoneSpace
    translationMode=BMM_Additive
    translationSpace=BCS_BoneSpace
    boneToModify {boneName=pack}
    translation {y=2}
   }
   n2 = "Transform (Modify) Bone - Bone: pack"!(ComponentPose=@n1, Alpha:float, Translation="0,0,0", Rotation=Get backpack_rotate(), Scale) -> (Pose) {
    rotationMode=BMM_Additive
    translationMode=BMM_Additive
    alphaBoolBlend {blendOption=QuarticInOut, blendOutTime=1}
    boneToModify {boneName=pack}
   }
   n3 = "Transform (Modify) Bone - Bone: pack"!(ComponentPose=@n2, bAlphaBoolEnabled:bool, Translation="0,-2,0", Rotation="0,0,10", Scale) -> (Pose) {
    alphaInputType=Bool
    rotationMode=BMM_Additive
    rotationSpace=BCS_BoneSpace
    translationMode=BMM_Additive
    translationSpace=BCS_BoneSpace
    alphaBoolBlend {blendOption=QuadraticInOut, blendOutTime=1}
    boneToModify {boneName=pack}
    rotation {roll=10}
    translation {y=-2}
   }
   n4 = "Transform (Modify) Bone - Bone: dyn_eyelid_top"(
    ComponentPose="Transform (Modify) Bone - Bone: aimtarget_eye"(ComponentPose=@n3, Alpha=1, Translation="0,0,0", Rotation="0,0,0", Scale="1,1,1") {translationMode=BMM_Additive, translationSpace=BCS_BoneSpace, boneToModify {boneName=aimtarget_eye}}
    Alpha=1
    Translation="0,0,0"
    Rotation="0,0,31"
    Scale="1,1,1"
   ) {
    rotationMode=BMM_Additive
    rotationSpace=BCS_BoneSpace
    alphaScaleBiasClamp {bInterpResult=true, interpSpeedDecreasing=15, interpSpeedIncreasing=15}
    boneToModify {boneName=dyn_eyelid_top}
    rotation {roll=31}
   }
   n5 = "Transform (Modify) Bone - Bone: dyn_eyelid_bttn"(ComponentPose=@n4, Alpha=1, Translation="0,0,0", Rotation="0,0,31", Scale="1,1,1") {
    rotationMode=BMM_Additive
    rotationSpace=BCS_BoneSpace
    alphaScaleBiasClamp {bInterpResult=true, interpSpeedDecreasing=15, interpSpeedIncreasing=15}
    boneToModify {boneName=dyn_eyelid_bttn}
    rotation {roll=31}
   }
   n6 = Look At - Bone: dyn_eye(ComponentPose=@n5, Alpha=1) {boneToModify {boneName=dyn_eye}, lookAtLocation=<0 100>, lookAtTarget {boneReference {boneName=aimtarget_eye}}}
   Output Pose(Result=Component To Local(ComponentPose=@n6)) {layerGroup=None}
  ]}
 EventGraph {nodes [
   Event BlueprintUpdateAnimation()
   > n1 = Sequence!(execute) -> (then_0, then_1, then_2:Exec, then_3:Exec)
   n2 = Branch(execute=@n1.then_2, Condition=Get rotateback() -> (rotateback:bool))
   n3 = GetOwnerSkeletalMeshComponent()
   n4 = BreakRotator(InRot=BreakTransform(InTransform=GetDeltaTransformFromRefPose(self=@n3, BoneName=spine_05, BaseName=spine_01)).Rotation)
   n5 = MapRangeClamped(Value=@n4.Yaw, InRangeA=-10, InRangeB=-30, OutRangeB=-10)
   n6 = MapRangeClamped(Value=@n4.Yaw, InRangeA=-10, InRangeB=-30, OutRangeA=10, OutRangeB=20)
   n7 = MapRangeClamped(Value=@n4.Yaw, InRangeA=-10, InRangeB=-30, OutRangeB=10)
   > Set backpack_rotate(backpack_rotate="Scale Rotator (Float)"(A:Rotator=MakeRotator(Roll=@n6, Pitch=@n7, Yaw=@n5), B:float=0.3) -> (ReturnValue:Rotator=~) {is=title K2Node_PromotableOperator})
   Set backpack_rotate(execute=@n2.else, backpack_rotate=MakeRotator(Pitch=@n7, Yaw=@n5))
   Delay(execute=@n1.then_1, Duration=RandomFloatInRange(Min=4, Max=10))
   > Set Blink(Blink:double=1) -> (Output_Get:double)
   > Delay(Duration=RandomFloatInRange(Min=0.2, Max=0.4))
   > Set Blink(Blink=0)
   Branch(execute=@n1.then_0, Condition=Get bEnemyPlayerEliminated() -> (bEnemyPlayerEliminated:bool))
   n8 = GetRealTimeSeconds()
   n9 = float * float(A=@n8, B=5)
   n10 = MakeVector(
    X=float * float(A="Cos (Radians)"!(A:double=@n9) -> (ReturnValue:double), B=20) {is=title K2Node_PromotableOperator}
    Y=0
    Z=float * float(A="Sin (Radians)"(A=@n9), B=20)
   )
   n11 = float * float(A=@n8, B=4) {is=title K2Node_CommutativeAssociativeBinaryOperator}
   n12 = MakeVector(Y=0, Z=float * float(A="Sin (Radians)"(A=@n11), B=20) {is=title K2Node_PromotableOperator})
   n13 = float * float(A=@n8, B=3.5) {is=title K2Node_CommutativeAssociativeBinaryOperator}
   n14 = MakeVector(X=float * float(A="Cos (Radians)"(A=@n13), B=20) {is=title K2Node_PromotableOperator}, Y=0)
   > Set EyeTranslate(EyeTranslate:Vector="Get (a copy)"(
     Array:Array of Vectors=Make Array!("[0]":Vector=@n10, "[1]":Vector=@n12, "[2]":Vector=@n14) -> (Array:Array of Vectors)
     Dimension 1=Get RandomEyeAnim() -> (RandomEyeAnim:int=0)
    ) -> (Output:Vector)) -> (Output_Get:Vector)
   Event OnBeginSkydiving()
   > Set rotateback(rotateback:bool=true) -> (Output_Get:bool)
   Event OnEndSkydiving()
   > Set rotateback(rotateback=false)
   n15 = Event BlueprintProcessAnimInputEvent()
   > Branch(Condition=IsNamed(self=@n15.AnimInputEvent, InEventName=aie.EnemyPlayerEliminated))
   > Set bEnemyPlayerEliminated(bEnemyPlayerEliminated:bool=true) -> (Output_Get:bool)
   > Set RandomEyeAnim(RandomEyeAnim:int=RandomInteger!(Max:int=3) -> (ReturnValue:int=0)) -> (Output_Get:int=0)
   > Delay(Duration=2)
   > Set bEnemyPlayerEliminated(bEnemyPlayerEliminated=false)
  ]}
]
variables [
 backpack_rotate {type=Rotator}
 bEnemyPlayerEliminated {type=boolean}
 Blink {type=number}
 EyeTranslate {type=Vector}
 gender female {type=boolean}
 RandomEyeAnim {type=integer}
 rotateback {type=boolean}
]
